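# Physics asset for the medium male BeetTread player body, covering its front and back loincloths and belt pieces.
# Capsules on the pelvis and thighs form the body volumes, while tiny base capsules link each cloth or belt chain to the pelvis.
# Two-segment dyn chains hang from those bases with narrow swing limits and mostly locked twist.
# The front loincloth uses boxes, while the back loincloth and belts use capsules.
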PhysicsAsset /BRCosmetics/Characters/Player/Male/Medium/Bodies/M_MED_BeetTread/Meshes/M_MED_BeetTread_Physics
bodies {
 base_back_loincloth_ [{shapeType=Capsule, length=0.505, radius=0.505, shapeName=base_back_loincloth__capsule, rotation=<rot 0 -90 0>}]
 base_belt [{shapeType=Capsule, length=0.505, radius=0.505, shapeName=base_belt_capsule, rotation=<rot 0 -90 0>}]
 base_front_loincloth [{shapeType=Capsule, length=0.505, radius=0.505, shapeName=base_front_loincloth_capsule, rotation=<rot 0 -90 0>}]
 dyn_back_loincloth_01 [{shapeType=Capsule, length=13.0875, radius=2.5, center=<4.9884 -1.0474 0.5193>}]
 dyn_back_loincloth_02 [{shapeType=Capsule, length=11.7728, radius=3.5, shapeName=dyn_front_loincloth_02_capsule, center=<7.0107 -1.5999 0.1186>}]
 dyn_belt_01_l [{shapeType=Capsule, length=2, radius=0.5, shapeName=dyn_belt_01_l_capsule, center=<1.9764 0 0>, rotation=<rot 90 0 0>}]
 dyn_belt_01_r [{shapeType=Capsule, length=2, radius=0.5, shapeName=dyn_belt_01_r_capsule, center=<1.5044 0 0>, rotation=<rot 90 0 0>}]
 dyn_belt_02_l [{shapeType=Capsule, length=2.2941, radius=0.75, shapeName=dyn_belt_02_l_capsule, center=<1.5584 0.0315 -0.0616>, rotation=<rot 90 0 0>}]
 dyn_belt_02_r [{
   shapeType=Capsule
   length=1.5562
   radius=1.5
   shapeName=dyn_belt_02_r_capsule
   center=<2.1463 -0.0993 0.0707>
   rotation=<rot 0 -90.5082 -79.0039>
  }]
 dyn_front_loincloth_01 [{shapeType=Box, extentX=10.5, extentY=3, extentZ=19.5, shapeName=dyn_front_loincloth_01_box, center=<4.7968 1.2485 0>}]
 dyn_front_loincloth_02 [{shapeType=Box, extentX=9.5, extentY=3.25, extentZ=20.5, shapeName=dyn_front_loincloth_01_box, center=<4.7174 1.0365 0>}]
 pelvis [{shapeType=Capsule, length=9.3588, radius=10.5539, center=<-1.639 -0.5843 0.0777>, rotation=<rot 0 89.4642 0.1544>}]
 thigh_l [{shapeType=Capsule, length=25.9911, radius=10.5, center=<-17.5005 -0.4602 -1.1396>, rotation=<rot 90 0 0>}]
 thigh_r [{shapeType=Capsule, length=26.4911, radius=10.5, center=<17.5005 0.6889 -1.1396>, rotation=<rot 90 0 0>}]
}
constraints [
 {bone1Name=thigh_r, bone2Name=pelvis, swing1LimitDegrees=45, swing2LimitDegrees=45, twistLimitDegrees=45}
 {
  bone1Name=dyn_front_loincloth_01
  bone2Name=base_front_loincloth
  swing1LimitDegrees=20
  swing2LimitDegrees=5
  twistLimitDegrees=0
  twistMotion=Locked
 }
 {
  bone1Name=dyn_back_loincloth_01
  bone2Name=base_back_loincloth_
  swing1LimitDegrees=25
  swing2LimitDegrees=10
  twistLimitDegrees=45
  twistMotion=Locked
 }
 {
  bone1Name=dyn_front_loincloth_02
  bone2Name=dyn_front_loincloth_01
  swing1LimitDegrees=10
  swing2LimitDegrees=5
  twistLimitDegrees=45
  twistMotion=Locked
 }
 {bone1Name=dyn_belt_02_r, bone2Name=dyn_belt_01_r, swing1LimitDegrees=20, swing2LimitDegrees=20, twistLimitDegrees=5}
 {bone1Name=dyn_belt_01_r, bone2Name=base_belt, swing1LimitDegrees=10, swing2LimitDegrees=10, twistLimitDegrees=5}
 {bone1Name=dyn_belt_02_l, bone2Name=dyn_belt_01_l, swing1LimitDegrees=20, swing2LimitDegrees=20, twistLimitDegrees=45, twistMotion=Locked}
 {bone1Name=dyn_belt_01_l, bone2Name=base_belt, swing1LimitDegrees=10, swing2LimitDegrees=10, twistLimitDegrees=45, twistMotion=Locked}
 {
  bone1Name=dyn_back_loincloth_02
  bone2Name=dyn_back_loincloth_01
  swing1LimitDegrees=15
  swing2LimitDegrees=10
  twistLimitDegrees=45
  twistMotion=Locked
 }
 {
  bone1Name=base_back_loincloth_
  bone2Name=pelvis
  swing1LimitDegrees=45
  swing1Motion=Free
  swing2LimitDegrees=45
  swing2Motion=Free
  twistLimitDegrees=45
  twistMotion=Free
 }
 {
  bone1Name=base_belt
  bone2Name=pelvis
  swing1LimitDegrees=45
  swing1Motion=Free
  swing2LimitDegrees=45
  swing2Motion=Free
  twistLimitDegrees=45
  twistMotion=Free
 }
 {
  bone1Name=base_front_loincloth
  bone2Name=pelvis
  swing1LimitDegrees=45
  swing1Motion=Free
  swing2LimitDegrees=45
  swing2Motion=Free
  twistLimitDegrees=45
  twistMotion=Free
 }
 {
  bone1Name=thigh_l
  bone2Name=pelvis
  swing1LimitDegrees=45
  swing1Motion=Free
  swing2LimitDegrees=45
  swing2Motion=Free
  twistLimitDegrees=45
  twistMotion=Free
 }
]
