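# Physics asset for the face accessory mesh of the BeetTread male medium player cosmetic, set up around a necktie.
# Capsule bodies cover the head and upper spine, with a tiny capsule on the necktie base and two side-by-side capsules on the dynamic necktie bone.
# The necktie bone has angular limits against its base, tightest in twist.
# The head and necktie base link to spine_05 with free rotation on all axes.
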PhysicsAsset /BRCosmetics/Characters/Player/Male/Medium/Bodies/M_MED_BeetTread/Meshes/Parts/M_MED_BeetTread_FaceAcc_Physics
bodies {
 base_necktie [{shapeType=Capsule, length=0.505, radius=0.505, shapeName=base_necktie_capsule, rotation=<rot 0 -90 0>}]
 dyn_necktie_01 [
  {shapeType=Capsule, length=8.4677, radius=1, shapeName=dyn_necktie_01_capsule, center=<5.6797 0.8201 1.1564>, rotation=<rot 80 179.9999 180>}
  {shapeType=Capsule, length=8.4677, radius=1, shapeName=dyn_necktie_01_capsule, center=<5.7551 0.4117 -1.5289>, rotation=<rot 80 -0.0001 0>}
 ]
 head [{shapeType=Capsule, length=14.4702, radius=7.0249, center=<2.7723 3.5564 0.0006>, rotation=<rot 0.0025 33.038 -89.9718>}]
 spine_05 [{shapeType=Capsule, length=17, radius=11.5, shapeName=spine_05_capsule, center=<0 -0.3582 0>, rotation=<rot 0 -90 0>}]
}
constraints [
 {bone1Name=dyn_necktie_01, bone2Name=base_necktie, swing1LimitDegrees=50, swing2LimitDegrees=20, twistLimitDegrees=5}
 {
  bone1Name=head
  bone2Name=spine_05
  swing1LimitDegrees=45
  swing1Motion=Free
  swing2LimitDegrees=45
  swing2Motion=Free
  twistLimitDegrees=45
  twistMotion=Free
 }
 {
  bone1Name=base_necktie
  bone2Name=spine_05
  swing1LimitDegrees=45
  swing1Motion=Free
  swing2LimitDegrees=45
  swing2Motion=Free
  twistLimitDegrees=45
  twistMotion=Free
 }
]
