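# Animation Blueprint for the BeetTread character's face accessory part, adding corrective posing and physics on top of the copied parent mesh pose.
# The event graph derives scaled head counter-rotation and blend alphas from head yaw against spine and neck, plus a front-end idle flag.
# These drive additive Transform Bone nodes on head, base_necktie and Alt_pusher, feeding a head-space RigidBody simulation with a custom physics asset and gravity override.
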AnimBlueprint /BRCosmetics/Characters/Player/Male/Medium/Bodies/M_MED_BeetTread/Meshes/Parts/M_MED_BeetTread_FaceAcc_AnimBP
target_skeleton=&/M_MED_BeetTread_FaceAcc_Skeleton
graphs [
 AnimGraph {nodes [
   "Transform (Modify) Bone - Bone: Alt_pusher"(Alpha=1, Translation="2,3,0", Rotation="0,-40,0", Scale="1,1,1") {
    rotationMode=BMM_Additive
    rotationSpace=BCS_BoneSpace
    translationMode=BMM_Additive
    translationSpace=BCS_BoneSpace
    alphaBoolBlend {blendOption=Linear}
    alphaScaleBiasClamp {bInterpResult=true, interpSpeedDecreasing=8, interpSpeedIncreasing=8}
    boneToModify {boneName=Alt_pusher}
    rotation {yaw=-40}
    translation=<2 3>
   }
   n1 = "Transform (Modify) Bone - Bone: head"(
    ComponentPose=Local To Component(LocalPose=Copy Pose From Mesh() {bUseAttachedParent=true})
    Alpha=1
    Translation="0,0,0"
    Rotation=Get head rot() -> (head rot:Rotator)
    Scale="1,1,1"
   ) {
    rotationMode=BMM_Additive
    rotationSpace=BCS_BoneSpace
    translationMode=BMM_Additive
    translationSpace=BCS_BoneSpace
    alphaBoolBlend {blendOption=Linear}
    boneToModify {boneName=head}
   }
   n2 = "Transform (Modify) Bone - Bone: head"!(ComponentPose=@n1, bAlphaBoolEnabled:bool=True, Translation="0,0,0", Rotation="0,-7.5,0", Scale="1,1,1") -> (Pose) {
    alphaInputType=Bool
    rotationMode=BMM_Additive
    rotationSpace=BCS_BoneSpace
    translationMode=BMM_Additive
    translationSpace=BCS_BoneSpace
    alphaBoolBlend {blendOption=Linear}
    boneToModify {boneName=head}
    rotation {yaw=-7.5}
   }
   n3 = "Transform (Modify) Bone - Bone: base_necktie"!(ComponentPose=@n2, bAlphaBoolEnabled:bool=True, Translation="-2,-2,0", Rotation="0,0,0", Scale="1,1,1") -> (Pose) {
    alphaInputType=Bool
    rotationMode=BMM_Additive
    rotationSpace=BCS_BoneSpace
    translationMode=BMM_Additive
    translationSpace=BCS_BoneSpace
    alphaBoolBlend {blendOption=Linear}
    boneToModify {boneName=base_necktie}
    translation=<-2 -2>
   }
   n4 = "Transform (Modify) Bone - Bone: Alt_pusher"(ComponentPose=@n3, Alpha=Get Headrotated_r() -> (Headrotated_r:double), Translation="3,1,0", Rotation="0,-40,0", Scale="1,1,1") {
    rotationMode=BMM_Additive
    rotationSpace=BCS_BoneSpace
    translationMode=BMM_Additive
    translationSpace=BCS_BoneSpace
    alphaBoolBlend {blendOption=Linear}
    alphaScaleBiasClamp {bInterpResult=true, interpSpeedDecreasing=20, interpSpeedIncreasing=20}
    boneToModify {boneName=Alt_pusher}
    rotation {yaw=-40}
    translation=<3 1>
   }
   n5 = "Transform (Modify) Bone - Bone: Alt_pusher"(ComponentPose=@n4, Alpha=Get Headrotated() -> (Headrotated:double), Translation="2.5,0,0", Rotation="0,40,0", Scale="1,1,1") {
    rotationMode=BMM_Additive
    rotationSpace=BCS_BoneSpace
    translationMode=BMM_Additive
    translationSpace=BCS_BoneSpace
    alphaBoolBlend {blendOption=Linear}
    alphaScaleBiasClamp {bInterpResult=true, interpSpeedDecreasing=20, interpSpeedIncreasing=20}
    boneToModify {boneName=Alt_pusher}
    rotation {yaw=40}
    translation {x=2.5}
   }
   n6 = "Transform (Modify) Bone - Bone: Alt_pusher"(ComponentPose=@n5, Alpha=Get FEIDle() -> (FEIDle:double), Translation="0,0,0", Rotation="0,45,0", Scale="1,1,1") {
    rotationMode=BMM_Additive
    rotationSpace=BCS_BoneSpace
    translationMode=BMM_Additive
    translationSpace=BCS_BoneSpace
    alphaBoolBlend {blendOption=Linear}
    alphaScaleBiasClamp {bInterpResult=true, interpSpeedDecreasing=20, interpSpeedIncreasing=20}
    boneToModify {boneName=Alt_pusher}
    rotation {yaw=45}
   }
   n7 = "Transform (Modify) Bone - Bone: base_necktie"!(ComponentPose=@n6, Alpha:float=Get HeadRotateAlpha() -> (HeadRotateAlpha:double), Translation="0,0,0", Rotation="0,-60,0", Scale="1,1,1") -> (Pose) {
    rotationMode=BMM_Additive
    rotationSpace=BCS_BoneSpace
    translationMode=BMM_Additive
    translationSpace=BCS_BoneSpace
    alphaBoolBlend {blendOption=Linear}
    boneToModify {boneName=base_necktie}
    rotation {yaw=-60}
   }
   n8 = RigidBody!(ComponentPose=@n7, Alpha, OverrideWorldGravity:Vector="0,0,-980", ExternalForce) -> (Pose) {
    bClampLinearTranslationLimitToRefPose=true
    bOverrideWorldGravity=true
    overridePhysicsAsset=&/M_MED_BeetTread_FaceAcc_Physics
    simulationSpace=BaseBoneSpace
    alphaBoolBlend {blendOption=Linear}
    baseBoneRef {boneName=head}
    componentAppliedLinearAccClamp=<3000 3000 3000>
    componentLinearAccScale=<0.4 0.3 0.3>
    componentLinearVelScale=<0.4 0.3 0.3>
    overrideWorldGravity {z=-980}
    simSpaceSettings {
     maxAngularAcceleration=1000
     maxAngularVelocity=3
     maxLinearAcceleration=700
     maxLinearVelocity=500
     velocityScaleZ=0
     worldAlpha=0.25
     externalLinearDragV=<0.2 0.2>
    }
   }
   Output Pose(Result=Component To Local(ComponentPose=@n8)) {layerGroup=None}
  ]}
 EventGraph {nodes [
   Event BlueprintUpdateAnimation()
   > n1 = Sequence!(execute) -> (then_0, then_1, then_2:Exec, then_3:Exec)
   > n2 = Gravity Override State Function DEPRECATED(CharacterPartAnimInstance=Self-Reference(), GameTimeInSeconds=GetGameTimeInSeconds(), GravityOverrideParams=Get Gravity Override Params())
   > Set GravityOverride(GravityOverride=@n2)
   n3 = GetOwnerSkeletalMeshComponent()
   n4 = BreakRotator(InRot=BreakTransform(InTransform=GetDeltaTransformFromRefPose(self=@n3, BoneName=head, BaseName=spine_05)).Rotation)
   n5 = Get Y percentage() -> (Y percentage:float)
   n6 = Get X percentage() -> (X percentage:float)
   n7 = float * float!(A=@n4.Yaw, B=-1, C:double=Get Z percentage() -> (Z percentage:float)) -> (ReturnValue)
   > Set head rot(head rot:Rotator=MakeRotator(Roll=float * float(A=@n4, B=-1, C=@n6), Pitch=float * float(A=@n4.Pitch, B=-1, C=@n5), Yaw=@n7)) -> (Output_Get:Rotator)
   > Set HeadRotateAlpha(HeadRotateAlpha:double=Remap Float(Input=@n7, OldRange=(X=0,Y=-5), NewRange=(X=0,Y=1), Clamp=true) -> (Output=~)) -> (Output_Get:double)
   n8 = GetOwnerSkeletalMeshComponent()
   n9 = BreakRotator(InRot=BreakTransform(InTransform=GetDeltaTransformFromRefPose(self=@n8, BoneName=head, BaseName=spine_05)).Rotation)
   Set Headrotated(
    execute=@n1.then_1
    Headrotated:double=FInterpEaseInOut!(A:double=0, B:double=MapRangeClamped(Value=@n9.Yaw, InRangeA=0, InRangeB=-40, OutRangeB=10), Alpha:double=0, Exponent:double=0) -> (ReturnValue:double)
   ) -> (Output_Get:double)
   n10 = GetOwnerSkeletalMeshComponent()
   n11 = BreakRotator(InRot=BreakTransform(InTransform=GetDeltaTransformFromRefPose(self=@n10, BoneName=head, BaseName=neck_01)).Rotation)
   Set Headrotated_r(
    execute=@n1.then_2
    Headrotated_r:double=FInterpEaseInOut(A=0, B=MapRangeClamped(Value=@n11.Yaw, InRangeA=0, InRangeB=80, OutRangeB=13), Alpha=0, Exponent=0)
   ) -> (Output_Get:double)
   n12 = OR Boolean(A=NOT Boolean(A=Get bIsSkydiveDiveMode()), B=NOT Boolean(A=Get bIsPlayingEmote()))
   n13 = Get bIsInFrontEnd()
   Set FEIDle(execute=@n1.then_3, FEIDle:double="To Float (Boolean)"(InBool=AND Boolean(A=@n13, B=@n12))) -> (Output_Get:double)
   "To Float (Boolean)"(InBool="float <= float"(B=-17) -> (ReturnValue=~))
   "To Float (Boolean)"(InBool="float >= float"(B=-1))
   n14 = GetDeltaTransformFromRefPose(self=GetOwnerSkeletalMeshComponent(), BoneName=head, BaseName=spine_05)
   n15 = FInterpEaseInOut(A=0, B=BreakRotator(InRot=BreakTransform(InTransform=@n14).Rotation).Yaw, Alpha=0, Exponent=0)
   Set Headrotated(Headrotated="To Float (Boolean)"(InBool="float <= float"(A=@n15, B=-17) -> (ReturnValue=~)))
   n16 = GetDeltaTransformFromRefPose(self=GetOwnerSkeletalMeshComponent(), BoneName=head, BaseName=spine_05)
   n17 = FInterpEaseInOut(A=BreakRotator(InRot=BreakTransform(InTransform=@n16).Rotation).Yaw, B=0, Alpha=0, Exponent=0)
   Set Headrotated_r(Headrotated_r="To Float (Boolean)"(InBool="float >= float"(A=@n17, B=-1)))
  ]}
]
variables [
 FEIDle {type=number}
 Gravity Override Params {type=GravityOverrideParamsStruct}
 GravityOverride {type=Vector}
 head rot {type=Rotator}
 HeadRotateAlpha {type=number}
 Headrotated {type=number}
 Headrotated_r {type=number}
 X percentage {type=number}
 Y percentage {type=number}
 Z percentage {type=number, default=0.3}
]
